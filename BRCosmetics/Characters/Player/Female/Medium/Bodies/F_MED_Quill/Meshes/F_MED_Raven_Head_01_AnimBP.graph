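# Animation Blueprint for the Raven head mesh of the Quill female body, adding secondary motion to its head feathers.
# It copies the attached parent's pose, then chains three spring-based AnimDynamics nodes weighted by clamped pawn 2D speed.
# Skydiving feeds Perlin noise from pawn height into one feather's external force, halved under parachute.
# When the playlist uses custom character parts, the head bone shrinks to 0.01 scale, effectively hiding it.
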AnimBlueprint /BRCosmetics/Characters/Player/Female/Medium/Bodies/F_MED_Quill/Meshes/F_MED_Raven_Head_01_AnimBP
target_skeleton=&/F_MED_Raven_Head_01Skeleton
graphs [
 AnimGraph {nodes [
   Anim Dynamics - Bone: dyn_feathers_head_02(Alpha=1, ExternalForce="0,0,0") {
    angularDampingOverride=0.98
    angularSpringConstant=25000
    bAngularSpring=true
    bGravityOverrideInSimSpace=true
    bLinearSpring=true
    bOverrideAngularDamping=true
    bOverrideLinearDamping=true
    gravityScale=0.1
    lODThreshold=1
    linearDampingOverride=0.97
    linearSpringConstant=350
    alphaBoolBlend {blendOption=Linear}
    alphaScaleBiasClamp {bClampResult=true, bInterpResult=true, interpSpeedDecreasing=2.5, interpSpeedIncreasing=2.5, outRange {max=0.1}}
    boundBone {boneName=dyn_feathers_head_02}
    componentAppliedLinearAccClamp=<2500 2500 2500>
    componentLinearAccScale=<1 1 1>
    componentLinearVelScale=<1 1 1>
    physicsBodyDefinitions [{
      sphereCollisionRadius=0
      boundBone {boneName=dyn_feathers_head_02}
      boxExtents=<50 50 50>
      constraintSetup {
       angularTargetAxis=AxisZ
       angularLimitsMax=<45 45 45>
       angularLimitsMin=<-45 -45 -45>
       angularTarget=<0 0 1>
       linearAxesMax=<0 0 10>
       linearAxesMin=<0 0 -10>
      }
      localJointOffset=<0 0 0>
     }]
   }
   n1 = "Transform (Modify) Bone - Bone: head"!(ComponentPose=Local To Component(LocalPose=Use cached pose 'PreScale Cache'()), Alpha=1) -> (Pose) {scaleMode=BMM_Replace, alphaBoolBlend {blendOption=Linear}, boneToModify {boneName=head}, scale=<0.01 0.01 0.01>}
   Output Pose(Result=Blend Poses by bool(
     BlendPose_0=Component To Local(ComponentPose=@n1)
     BlendPose_1=Use cached pose 'PreScale Cache'()
     BlendTime_0=0
     BlendTime_1=0
     bActiveValue=Get GetPlayerListUsesCustomCharacterParts() -> (GetPlayerListUsesCustomCharacterParts:bool)
    ) {blendType=Linear}) {layerGroup=None}
   n2 = Get PawnSpeed2D()
   n3 = Anim Dynamics - Bone: dyn_feathers_head_02(
    ComponentPose=Local To Component(LocalPose=Copy Pose From Mesh() {bUseAttachedParent=true})
    Alpha=@n2
    ExternalForce=Get falling_anim_dyn_up_force_vec3() -> (falling_anim_dyn_up_force_vec3:Vector)
   ) {
    angularDampingOverride=0.98
    angularSpringConstant=325
    bAngularSpring=true
    bGravityOverrideInSimSpace=true
    bLinearSpring=true
    bOverrideAngularDamping=true
    bOverrideLinearDamping=true
    gravityScale=0
    lODThreshold=1
    linearDampingOverride=0.98
    linearSpringConstant=325
    simulationSpace=Actor
    alphaBoolBlend {blendOption=Linear}
    alphaScaleBiasClamp {
     bClampResult=true
     bInterpResult=true
     clampMax=0.25
     clampMin=0.05
     interpSpeedDecreasing=0.5
     interpSpeedIncreasing=0.5
     outRange {max=0.1}
    }
    boundBone {boneName=dyn_feathers_head_02}
    physicsBodyDefinitions [{
      sphereCollisionRadius=0
      boundBone {boneName=dyn_feathers_head_02}
      boxExtents=<30 30 30>
      constraintSetup {
       angularTargetAxis=AxisY
       twistAxis=AxisZ
       angularLimitsMax=<15 15 15>
       angularLimitsMin=<-15 -15 -15>
       angularTarget=<0 0 0>
       linearAxesMax=<10 10 10>
       linearAxesMin=<-10 -10 -10>
      }
      localJointOffset=<0 0 0>
     }]
   }
   n4 = Anim Dynamics - Bone: dyn_feathers_head_01(ComponentPose=@n3, Alpha=@n2, ExternalForce="0,0,0") {
    angularDampingOverride=0.97
    angularSpringConstant=275
    bAngularSpring=true
    bGravityOverrideInSimSpace=true
    bLinearSpring=true
    bOverrideAngularDamping=true
    bOverrideLinearDamping=true
    gravityScale=0.15
    lODThreshold=1
    linearDampingOverride=0.97
    linearSpringConstant=275
    simulationSpace=Actor
    alphaBoolBlend {blendOption=Linear}
    alphaScaleBiasClamp {
     bClampResult=true
     bInterpResult=true
     clampMax=0.25
     clampMin=0.05
     interpSpeedDecreasing=0.5
     interpSpeedIncreasing=0.5
     outRange {max=0.1}
    }
    boundBone {boneName=dyn_feathers_head_01}
    physicsBodyDefinitions [{
      sphereCollisionRadius=0
      boundBone {boneName=dyn_feathers_head_01}
      boxExtents=<30 30 30>
      constraintSetup {
       angularTargetAxis=AxisY
       twistAxis=AxisZ
       angularLimitsMax=<15 15 15>
       angularLimitsMin=<-15 -15 -15>
       angularTarget=<0 0 0>
       linearAxesMax=<10 10 10>
       linearAxesMin=<-10 -10 -10>
      }
      localJointOffset=<0 0 0>
     }]
   }
   n5 = Anim Dynamics - Bone: dyn_feathers_head_03(ComponentPose=@n4, Alpha=@n2, ExternalForce="0,0,0") {
    angularDampingOverride=0.96
    angularSpringConstant=300
    bAngularSpring=true
    bGravityOverrideInSimSpace=true
    bLinearSpring=true
    bOverrideAngularDamping=true
    bOverrideLinearDamping=true
    gravityScale=0.1
    lODThreshold=1
    linearDampingOverride=0.96
    linearSpringConstant=300
    simulationSpace=Actor
    alphaBoolBlend {blendOption=Linear}
    alphaScaleBiasClamp {
     bClampResult=true
     bInterpResult=true
     clampMax=0.25
     clampMin=0.05
     interpSpeedDecreasing=0.5
     interpSpeedIncreasing=0.5
     outRange {max=0.1}
    }
    boundBone {boneName=dyn_feathers_head_03}
    physicsBodyDefinitions [{
      sphereCollisionRadius=0
      boundBone {boneName=dyn_feathers_head_03}
      boxExtents=<30 30 30>
      constraintSetup {
       angularTargetAxis=AxisY
       twistAxis=AxisZ
       angularLimitsMax=<15 15 15>
       angularLimitsMin=<-15 -15 -15>
       angularTarget=<0 0 0>
       linearAxesMax=<10 10 10>
       linearAxesMin=<-10 -10 -10>
      }
      localJointOffset=<0 0 0>
     }]
   }
   Save cached pose 'PreScale Cache'(Pose=Component To Local(ComponentPose=@n5)) {cacheName=PreScale Cache}
  ]}
 EventGraph {nodes [
   Event BlueprintUpdateAnimation()
   n1 = Get PawnLastLocation() -> (PawnLastLocation:Vector)
   n2 = SelectFloat(
    A=Make Perlin Noise and Remap!(Value:float=BreakVector(InVec=@n1).Z, RangeOutMin:float=-25000, RangeOutMax:float=25000) -> (ReturnValue:float)
    B=0
    bPickA=Get bIsSkyDiving()
   )
   n3 = SelectFloat(A=float * float(A=@n2, B=0.5), B=@n2, bPickA=Get bIsParachuteOpened())
   > Set falling_anim_dyn_up_force_vec3(falling_anim_dyn_up_force_vec3:Vector=MakeVector(X=@n3, Y=@n3)) -> (Output_Get:Vector)
   Event BlueprintInitializeAnimation()
   > Set GetPlayerListUsesCustomCharacterParts(GetPlayerListUsesCustomCharacterParts:bool=GetPlaylistUsesCustomCharacterParts!() -> (ReturnValue:bool)) -> (Output_Get:bool)
   TryGetPawnOwner()
  ]}
]
variables [
 falling_anim_dyn_up_force_vec3 {type=Vector}
 GetPlayerListUsesCustomCharacterParts {type=boolean}
]
